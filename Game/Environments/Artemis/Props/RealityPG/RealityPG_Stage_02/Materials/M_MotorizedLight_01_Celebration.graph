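# Material for a celebration motorized stage light prop, with animated, color-cycling emissive light patterns over textured base color and normal.
# Customized UV1 selects cells in a 64-grid pattern atlas, with a time-driven sine stepping through the first few frames.
# The pattern, masked by mask-texture green, is tinted red, hue-shifted by a second sine and scaled for emissive.
# Attributes pass through MF_GameWideShared; a cosine If branch is left unconnected.
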Material /Game/Environments/Artemis/Props/RealityPG/RealityPG_Stage_02/Materials/M_MotorizedLight_01_Celebration
graph {
 properties {
  bUseMaterialAttributes=true
  bUsedWithMeshDeformer=true
  bUsedWithNanite=true
  bUsedWithSkeletalMesh=true
  numCustomizedUVs=2
  pixelDepthOffsetMode=PDOM_Legacy
  usedShadingModels=Default Lit
 }
 nodes [
  n1 = Multiply(A=ScalarParameter() {defaultValue=0.5, parameterName=Phase Speed}, B=Time())
  n2 = Multiply(A=ScalarParameter() {defaultValue=2}, B=Divide(A=@n1))
  n3 = MaterialFunctionCall!(Sine Phase=@n1, Value 1=Constant() {r=1}, Value 2=Constant() {r=4}) {materialFunction=/Engine/Functions/Engine_MaterialFunctions02/Utility/Sine_Remapped}
  If(A=Clamp(None=Cosine(None=@n2)), B=Constant(), "A > B"=@n3.Remapped Sine, "A == B"=Constant() {r=5}, "A < B"=@n3.Remapped Sine)
  n4 = TextureSample() {texture=/Game/Athena/Environments/Festivus/MotorizedLights/T_MotorizedLight_01_M}
  n5 = ScalarParameter() {defaultValue=10, parameterName=Emissive Brightness}
  n6 = ScalarParameter() {defaultValue=0.3, parameterName=HueShiftSpeed}
  n7 = Time()
  n8 = TextureSample(UVs=TextureCoordinate() {coordinateIndex=1}) {texture=/Game/Athena/Environments/Festivus/MotorizedLights/T_LightPatterns_01_M}
  n9 = MaterialFunctionCall(
   Hue Shift Percentage=Sine(None=Multiply(A=@n6, B=@n7))
   Texture=Multiply(A=Constant3Vector() {constant=<rgba 1 0 0 0>}, B=Multiply(A=@n8.RGB, B=@n4.G)) {constB=10}
  ) {materialFunction=&HueShift}
  n10 = TextureCoordinate() {coordinateIndex=1}
  n11 = Constant() {desc=Rows, r=64}
  n12 = Divide(A=Constant() {r=1}, B=@n11)
  n13 = AppendVector(A=Multiply(A=ComponentMask(None=@n10) {r=true}, B=@n12), B=Multiply(A=ComponentMask(None=@n10) {g=true}, B=@n12))
  n14 = Constant() {desc=FrameNumber, r=1}
  n15 = Constant()
  n16 = Clamp(None=Constant() {r=1})
  n17 = Floor(None=LinearInterpolate(A=@n14, B=LinearInterpolate(A=@n15, B=@n3.Remapped Sine, Alpha=@n16) {constA=16, constB=21}, Alpha=@n14))
  n18 = Divide(A=@n17, B=@n11)
  n19 = AppendVector(A=Multiply(A=Fmod(A=@n17, B=@n11), B=@n12), B=Multiply(A=Floor(None=@n18), B=@n12))
  MaterialFunctionCall!(In=MakeMaterialAttributes!(
    BaseColor=TextureSample() {texture=/Game/Athena/Environments/Festivus/MotorizedLights/T_MotorizedLight_01_D}.RGB
    Metallic
    Specular
    Roughness=@n4.B
    Anisotropy
    EmissiveColor=Multiply(A=@n5, B=@n9.Result) {constB=150}
    Opacity
    OpacityMask
    Normal=TextureSample() {samplerType=SAMPLERTYPE_Normal, texture=/Game/Athena/Environments/Festivus/MotorizedLights/T_MotorizedLight_01_N}.RGB
    Tangent
    WorldPositionOffset
    SubsurfaceColor
    ClearCoat
    ClearCoatRoughness
    AmbientOcclusion=@n4.B
    Refraction
    CustomizedUVs_0
    CustomizedUVs_1=Add(A=@n13, B=@n19)
    CustomizedUVs_2
    CustomizedUVs_3
    CustomizedUVs_4
    CustomizedUVs_5
    CustomizedUVs_6
    CustomizedUVs_7
    PixelDepthOffset
    ShadingModel
    Displacement
   )) {materialFunction=/Game/Athena/WinterArt/Materials/MF_GameWideShared}
  Subtract(A=@n17)
 ]
}
